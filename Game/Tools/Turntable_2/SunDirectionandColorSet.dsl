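; Turntable tool Blueprint that pushes sun direction and color into the shared FortniteMaterialParameters collection.
; An Arrow component supplies the normalized forward vector for sunlight, and a SunLightColor variable supplies the color.
; The sun and moon model vector is the arrow's roll, pitch and yaw normalized as a vector, not a true direction.
; The construction script writes all three parameters, and BeginPlay repeats this after a short delay; Tick and overlap are empty.
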(event EventBeginPlay
  (bind _arrow (Variables|Default|GetArrow))
  (Utilities|FlowControl|RetriggerableDelay 0.2)
  (Rendering|Material|SetVectorParameterValue "/Game/Packages/Fortress_SharedMaterials/GlobalMaterialParameters/FortniteMaterialParameters.FortniteMaterialParameters" "SunlightDirectionalVector" (Math|Conversions|ToLinearColor(Vector) (Math|Vector|Normalize (Transformation|GetForwardVector _arrow))) :!class /Script/Engine.KismetMaterialLibrary)
  (Rendering|Material|SetVectorParameterValue "/Game/Packages/Fortress_SharedMaterials/GlobalMaterialParameters/FortniteMaterialParameters.FortniteMaterialParameters" "SunAndMoonModelDirectionalVector" (Math|Conversions|ToLinearColor(Vector) (Math|Vector|Normalize (Math|Vector|MakeVector (.roll (Transformation|GetWorldRotation _arrow)) (.pitch (Transformation|GetWorldRotation _arrow)) (.yaw (Transformation|GetWorldRotation _arrow))))) :!class /Script/Engine.KismetMaterialLibrary)
  (Rendering|Material|SetVectorParameterValue "/Game/Packages/Fortress_SharedMaterials/GlobalMaterialParameters/FortniteMaterialParameters.FortniteMaterialParameters" "SunLightColor" (Variables|Default|GetSunLightColor) :!class /Script/Engine.KismetMaterialLibrary))

(event Collision|EventActorBeginOverlap
  (params (OtherActor :type Actor%20Object%20Reference)))

(event EventTick
  (params (DeltaSeconds :type Float%20%28single-precision%29)))

(fn ConstructionScript
  (bind _arrow (Variables|Default|GetArrow))
  (Rendering|Material|SetVectorParameterValue "/Game/Packages/Fortress_SharedMaterials/GlobalMaterialParameters/FortniteMaterialParameters.FortniteMaterialParameters" "SunlightDirectionalVector" (Math|Conversions|ToLinearColor(Vector) (Math|Vector|Normalize (Transformation|GetForwardVector _arrow))) :!class /Script/Engine.KismetMaterialLibrary)
  (Rendering|Material|SetVectorParameterValue "/Game/Packages/Fortress_SharedMaterials/GlobalMaterialParameters/FortniteMaterialParameters.FortniteMaterialParameters" "SunAndMoonModelDirectionalVector" (Math|Conversions|ToLinearColor(Vector) (Math|Vector|Normalize (Math|Vector|MakeVector (.roll (Transformation|GetWorldRotation _arrow)) (.pitch (Transformation|GetWorldRotation _arrow)) (.yaw (Transformation|GetWorldRotation _arrow))))) :!class /Script/Engine.KismetMaterialLibrary)
  (Rendering|Material|SetVectorParameterValue "/Game/Packages/Fortress_SharedMaterials/GlobalMaterialParameters/FortniteMaterialParameters.FortniteMaterialParameters" "SunLightColor" (Variables|Default|GetSunLightColor) :!class /Script/Engine.KismetMaterialLibrary))
